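; State Tree task that marks an encounter as failed when its state is entered.
; On EnterState it finds the EncounterManagerComponent on the context actor.
; It then calls HandleEncounterFailure on that component, passing the task's Reward Index variable.
; The task has no other logic and does not check that the component exists.
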(event EventEnterState
  (params (Transition :type State%20Tree%20Transition%20Result%20Structure%20%28by%20ref%29))
  (Class|ScenarioManagerComponent|HandleEncounterFailure (Actor|GetComponentbyClass (Variables|Context|GetActor) "/Script/EncountersRuntime.EncounterManagerComponent") (Variables|Default|GetReward%20Index)))
